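; Swamp lily pad prop that bounces pawns, hoverboards and push cannons touching Jump_Collision.
; Pawns not downed or already launched get velocity reset, a forward-and-up launch, launched effect, fall-damage immunity, animation and feedback.
; A capsule trace lowers pawn and hoverboard bounces under ceilings; delayed DoOnce resets aim to prevent doubled sounds.
; Push cannons are stored in a variable whose OnRep applies a clamped, velocity-based impulse plus spin thrust.
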(event Custom|Reset%20Bounce%20Sound
  ;[1 1 1] Keeps bounce sound from double triggering (hopefully)
  (Utilities|FlowControl|DoOnce false
    (:Completed
      (GameplayCue|ExecuteGameplayCueOnActor(Burst) self (Variables|Default|GetGC_JumpedOn) (Ability|GameplayCue|MakeGameplayCueParameters 0.0)))))

(event OnComponentBeginOverlap(Jump_Collision)
  (params (OverlappedComponent :type Primitive%20Component%20Object%20Reference) (OtherActor :type Actor%20Object%20Reference) (OtherComp :type Primitive%20Component%20Object%20Reference) (OtherBodyIndex :type Integer) (bFromSweep :type Boolean) (SweepResult :type Hit%20Result%20Structure%20%28by%20ref%29))
  (bind (_returnvalue_x _returnvalue_y) (Math|Vector|BreakVector (* (Transformation|GetActorForwardVector (Variables|Default|GetOverlapPawn)) (Variables|Default|GetBounceForwardVelocity))))
  (bind _overlappawn (Variables|Default|GetOverlapPawn))
  (bind _overlapcontroller (Variables|Default|GetOverlapController))
  (bind _asfort_pawn (Utilities|Casting|CastToFortPawn OtherActor)
    (:then
      (bind _output_get (Variables|Default|SetOverlapPawn _asfort_pawn))
      (if (not (or (GameplayTags|HasMatchingGameplayTag (Ability|GetAbilitySystemComponent _output_get) "(TagName=\"Gameplay.Action.Player.Launched\")") (Variables|/Script/FortniteGame.FortPawn|GetbIsDBNO _output_get)))
        (bind _asfort_player_controller (Utilities|Casting|CastToFortPlayerController (Pawn|GetController _output_get))
          (:then
            (Variables|Default|SetOverlapController _asfort_player_controller))
          (:CastFailed
            _))
        ;[1 1 1] Ceiling check
        (bind _success (CallFunction|CeilingCheck))
        (if _success
          (Variables|Default|SetBounceHeight (Variables|Default|GetBounceHeightLowCeiling))
          (Variables|Default|SetJumpHeight 100)
          (else
            (Variables|Default|SetBounceHeight (Variables|Default|GetBounceHeightDefault))
            (Variables|Default|SetJumpHeight 4.5e+02)))
        (Variables|Default|SetFinalBounceJumpHeight (Variables|Default|GetBounceHeight))
        (bind _asfort_player_pawn (Utilities|Casting|CastToFortPlayerPawn (Variables|Default|GetOverlapPawn))
          (:then
            (RemoteControlPawn|ForceDetachfromRemoteControlledPawn _asfort_player_pawn))
          (:CastFailed
            _))
        (Variables|/Script/Engine.MovementComponent|SetVelocity "0, 0, 0" (Variables|/Script/Engine.Character|GetCharacterMovement (Variables|Default|GetOverlapPawn))) ; so that launch distance is consistent
        (Character|LaunchCharacter (Variables|Default|GetOverlapPawn) (Math|Vector|MakeVector _returnvalue_x _returnvalue_y (Variables|Default|GetFinalBounceJumpHeight)) false true)
        (Ability|SendGameplayEventtoActor _asfort_pawn "(TagName=\"Event.Lilypad.Bounce\")")
        (GameplayEffects|ApplyGameplayEffectToSelf :self (Ability|GetAbilitySystemComponent _overlappawn) :GameplayEffectClass (Variables|Default|GetGE_PlayerLaunched) :Level 0.0)
        (Variables|/Script/FortniteGame.FortPawn|SetbIgnoreNextFallingDamage true _overlappawn)
        (Utilities|FlowControl|Sequence
          (:then_0
            (CallFunction|BounceAnim))
          (:then_1
            ;[1 1 1] Keeps bounce sound from double triggering (hopefully)
            (Utilities|FlowControl|DoOnce false
              (:Completed
                (GameplayCue|ExecuteGameplayCueOnActor(Burst) self (Variables|Default|GetGC_JumpedOn) (Ability|GameplayCue|MakeGameplayCueParameters 0.0)))))
          (:then_2
            (Utilities|IsValid _overlapcontroller
              (:Is%20Valid
                (Game|Feedback|LocalPlayForceFeedback _overlapcontroller (Variables|Default|GetLaunchForceFeedback) false "TireLaunch"))
              (:Is%20Not%20Valid
                _))
            (Variables|Default|SetOverlapPawn 0)
            (Variables|Default|SetOverlapController 0)
            ;[1 1 1] Reset Bounce sound, so sound cue can be heard if "X" times bounce happens
            (Utilities|FlowControl|Delay 0.1)
            (CallFunction|ResetBounceSound)))))
    (:CastFailed
      ;[1 1 1] Check for Collision with a Push Cannon
      (bind _returnvalue (GameplayTags|HasAnyMatchingGameplayTags(Message) OtherActor (Variables|Default|GetCannon%20Tag)))
      (if _returnvalue
        (bind _asfort_athena_vehicle (Utilities|Casting|CastToFortAthenaVehicle OtherActor))
        (Variables|Default|SetPush%20Cannon%20Vehicle _asfort_athena_vehicle)
        (else
          ;[1 1 1] Ceiling check
          (bind _success_2 (CallFunction|CeilingCheck))
          (if _success_2
            (Variables|Default|SetBounceHeight (Variables|Default|GetBounceHeightLowCeiling))
            (Variables|Default|SetJumpHeight 100)
            (else
              (Variables|Default|SetBounceHeight (* (Variables|Default|GetBounceHeightDefault) 1.25))
              (Variables|Default|SetJumpHeight 600)))
          (Variables|Default|SetFinalBounceJumpHeight (Variables|Default|GetBounceHeight))
          (bind _success_3 (Vehicles|LaunchHoverboard OtherActor (Math|Vector|MakeVector 0.0 0.0 (Variables|Default|GetFinalBounceJumpHeight)) "TireLaunch"))
          (if _success_3
            (Utilities|FlowControl|Sequence
              (:then_0
                ;[1 1 1] Keeps bounce sound from double triggering (hopefully)
                (Utilities|FlowControl|DoOnce false
                  (:Completed
                    (Audio|PlaySoundatLocation (Variables|Default|GetLilypadSound) (Transformation|GetActorLocation OtherActor)))))
              (:then_1
                ;[1 1 1] Reset Bounce sound, so sound cue can be heard if "X" times bounce happens
                (Utilities|FlowControl|Delay 0.1)
                (CallFunction|ResetBounceSound)))))))))

(event Custom|BounceAnim
  (CallFunction|StartBounceAnimation (Utilities|Time|GetTimeSeconds) 3e+02 (Math|Vector|MakeVector 0.0 0.0 1.0) (Transformation|GetActorLocation)))

(event Custom|CustomEvent
  ;[1 1 1] Keeps bounce sound from double triggering (hopefully)
  (Utilities|FlowControl|DoOnce false
    (:Completed
      (Audio|PlaySoundatLocation (Variables|Default|GetLilypadSound) (Transformation|GetActorLocation (stale OnComponentBeginOverlap(Jump_Collision) OtherActor))))))

(event Custom|Launch%20Cannon
  (bind _push_cannon_vehicle (Variables|Default|GetPush%20Cannon%20Vehicle))
  ;[1 1 1] Launches Push Cannon
  (Audio|BroadcastAbilityImpactatLocation 0 _push_cannon_vehicle (CallFunction|CalculateVehicleImpulse _push_cannon_vehicle (Variables|Default|GetLaunch%20Multiplier)) 777.0)
  (Audio|BroadcastAbilitySpinThrustatLocation 0 _push_cannon_vehicle (* (Math|Vector|SelectVector (Math|Vector|MirrorVectorbyNormal (Vehicle|GetVehicleRightVector _push_cannon_vehicle) (Vehicle|GetVehicleRightVector _push_cannon_vehicle)) (Vehicle|GetVehicleRightVector _push_cannon_vehicle) (> (Vehicle|GetForwardSpeedKmh _push_cannon_vehicle) 0.0)) (Variables|Default|GetSpin%20Thrust%20Multiplier)) 777.0))

(fn CalculateVehicleImpulse
  (params (Actor :type Actor%20Object%20Reference) (Multiplier :type Float%20%28double-precision%29))
  (return (Math|Vector|ClampVectorSize (Math|Vector|MakeVector (* (.x (Transformation|GetVelocity Actor)) Multiplier) (* (.y (Transformation|GetVelocity Actor)) Multiplier) (Math|Float|Clamp(Float) (* Multiplier (* (Math|Vector|VectorLength (Transformation|GetVelocity Actor)) 0.5)) 5e+02 1.2e+03)) 1.5e+03 2.5e+03)))

(fn CeilingCheck
  (bind _overlappawn (Variables|Default|GetOverlapPawn)
    (:then
      (Utilities|Array|Add (LocalVariables|GetActorsToIgnore) _overlappawn)
      (Utilities|Array|AppendArray (LocalVariables|GetActorsToIgnore) (Actor|GetAttachedActors _overlappawn true false))
      (for _array_element (Actor|GetAttachedActors _overlappawn true false)
        (Utilities|Array|AppendArray (LocalVariables|GetActorsToIgnore) (Actor|GetAttachedActors _array_element true false))))
    (:else
      _))
  (bind (_outhit _returnvalue) (Collision|CapsuleTraceByChannel (Math|Vector|MakeVector (.x (Transformation|GetWorldLocation (Variables|Transformation|GetRootComponent))) (.y (Transformation|GetWorldLocation (Variables|Transformation|GetRootComponent))) (+ (.z (Transformation|GetWorldLocation (Variables|Transformation|GetRootComponent))) 200)) (Math|Vector|MakeVector (.x (Transformation|GetWorldLocation (Variables|Transformation|GetRootComponent))) (.y (Transformation|GetWorldLocation (Variables|Transformation|GetRootComponent))) (+ (.z (Transformation|GetWorldLocation (Variables|Transformation|GetRootComponent))) 340)) 50 100 "TraceTypeQuery9" false (LocalVariables|GetActorsToIgnore)))
  (return _returnvalue))

(fn OnRep_PushCannonVehicle
  (CallFunction|LaunchCannon))

(fn ConstructionScript)
